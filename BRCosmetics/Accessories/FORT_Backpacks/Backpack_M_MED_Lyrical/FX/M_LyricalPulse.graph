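# Additive unlit Niagara sprite material for the Lyrical backpack's audio-visualizer-style pulse, tinted by particle color.
# Spinning radial plasma noise in an amplitude-sized circle combines with a mirrored horizontal waveform built from sine/cosine and panning wave textures.
# PanPosition and IncomingAmplitude dynamic parameters drive scrolling, size and intensity.
# A center cutout, particle alpha and forward-vector cone mask the result; a separate view-transformed vector computes amplitude-scaled chroma push.
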Material /BRCosmetics/Accessories/FORT_Backpacks/Backpack_M_MED_Lyrical/FX/M_LyricalPulse
graph {
 properties {bUsedWithNiagaraSprites=true, blendMode=BLEND_Additive, pixelDepthOffsetMode=PDOM_Legacy, shadingModel=MSM_Unlit, usedShadingModels=Unlit}
 nodes [
  n1 = ParticleColor()
  MaterialFunctionCall(Emissive=Multiply(A=@n1.RGB, B=VectorParameter() {parameterName=TintColor, defaultValue=<rgba 1 1 1 0>}.RGB)) {materialFunction=&MF_FX_EmissiveAdjustments}
  n2 = VectorParameter() {parameterName=ChromaPushDir, defaultValue=<rgba 1 0 0 0>}
  Multiply(
   A=Multiply(
    A=DynamicParameter() {paramNames [PanPosition, IncomingAmplitude, ChromaPushAmount, Param4]}.IncomingAmplitude
    B=Transform(None=@n2.RGB) {transformSourceType=TRANSFORMSOURCE_View}
   )
   B=ScalarParameter() {defaultValue=1, parameterName=ChromaPushAmt}
  )
  n3 = Transform(None=VectorParameter() {parameterName=ForwardVector, defaultValue=<rgba 0 1 0 0>}.RGB) {transformSourceType=TRANSFORMSOURCE_Local}
  n4 = ScalarParameter() {defaultValue=-10, parameterName=VisibilityConeOffset}
  n5 = Subtract(A=Add(A=Multiply(A=@n3, B=@n4), B=WorldPosition().XYZ), B=ObjectPositionWS())
  n6 = Multiply(A=DotProduct(A=@n3, B=Normalize(VectorInput=@n5)), B=ScalarParameter() {defaultValue=3, parameterName=VisibilityConeHardness})
  n7 = OneMinus(None=SphereMask(A=TextureCoordinate(), B=Constant2Vector() {g=0.5, r=0.5}, Radius=Constant() {r=0.2}, Hardness=Constant() {r=0.3}))
  n8 = MaterialFunctionCall() {materialFunction=&VectorToRadialValue}
  n9 = AppendVector(A=ScalarParameter() {defaultValue=1, parameterName=TileU}, B=ScalarParameter() {defaultValue=3, parameterName=TileV})
  n10 = Multiply(A=Time(), B=ScalarParameter() {defaultValue=0.1, parameterName=SpinSpeed})
  n11 = Multiply(A=DynamicParameter() {paramNames [PanPosition, IncomingAmplitude, ChromaPushAmount, Param4]}.PanPosition) {constB=-1}
  n12 = TextureCoordinate()
  n13 = TextureSampleParameter2D!(
   UVs=Add(A=Multiply(A=@n8.Radial Coordinates, B=@n9), B=AppendVector(A=@n10, B=@n11))
   "DDX(UVs)"=DDX(Value=@n12)
   "DDY(UVs)"=DDY(Value=@n12)
   Apply View MipBias
  ) {mipValueMode=TMVM_Derivative, parameterName=NoiseTexture, texture=/Game/Effects/Fort_Effects/Textures/T_FX_SpringerPlasma}
  n14 = DynamicParameter() {paramNames [PanPosition, IncomingAmplitude, ChromaPushAmount, Param4]}
  n15 = LinearInterpolate(
   A=ScalarParameter() {defaultValue=0.2, parameterName=CircularElementOpacityLow}
   B=ScalarParameter() {defaultValue=0.05, parameterName=CircularElementOpacityHigh}
   Alpha=Saturate(None=Multiply(A=@n14.IncomingAmplitude) {constB=10})
  )
  n16 = TextureCoordinate()
  n17 = Constant2Vector() {g=0.5, r=0.5}
  n18 = LinearInterpolate(
   A=ScalarParameter() {defaultValue=0.3, parameterName=CircleRadiusLow}
   B=ScalarParameter() {defaultValue=0.5, parameterName=CircleRadiusHigh}
   Alpha=DynamicParameter() {paramNames [PanPosition, IncomingAmplitude, ChromaPushAmount, Param4]}.IncomingAmplitude
  ) {constA=0.3, constB=0.5}
  n19 = ScalarParameter() {defaultValue=0.5, parameterName=CircleHardness}
  n20 = ComponentMask(None=TextureCoordinate()) {r=true}
  n21 = OneMinus(None=@n20)
  n22 = ScalarParameter() {defaultValue=-0.15, parameterName=sqz}
  n23 = DynamicParameter() {paramNames [PanPosition, IncomingAmplitude, ChromaPushAmount, Param4]}
  n24 = Multiply(A=Add(A=Multiply(A=@n21, B=@n20), B=LinearInterpolate(A=@n22, Alpha=@n23.IncomingAmplitude) {constA=-0.1, constB=0})) {constB=6}
  n25 = TextureCoordinate()
  n26 = Abs(None=Add(A=ComponentMask(None=@n25) {r=true}) {constB=-0.5})
  n27 = Multiply(A=@n26, B=@n26)
  n28 = Power(Base=@n26) {constExponent=0.7}
  n29 = ComponentMask(None=TextureCoordinate()) {g=true}
  n30 = Abs(None=Multiply(A=Add(A=@n29) {constB=-0.5}) {constB=2})
  n31 = Multiply(A=@n30, B=@n30)
  n32 = Multiply(A=AppendVector(A=LinearInterpolate(A=@n27, B=@n28, Alpha=@n31), B=@n29), B=Constant2Vector() {g=0.01, r=12})
  n33 = DynamicParameter() {paramNames [PanPosition, IncomingAmplitude, ChromaPushAmount, Param4]}
  n34 = Constant()
  n35 = TextureSample(UVs=Add(A=Panner(Coordinate=@n32) {speedY=-0.752}, B=AppendVector(A=@n33.PanPosition, B=@n34))) {texture=/Game/Effects/Fort_Effects/Textures/T_Water_Tile1}
  n36 = DynamicParameter() {paramNames [PanPosition, IncomingAmplitude, ChromaPushAmount, Param4]}
  n37 = Constant()
  n38 = TextureSample(UVs=Add(A=Panner(Coordinate=@n32) {speedY=0.1}, B=AppendVector(A=@n36.PanPosition, B=@n37))) {texture=/Game/Effects/Fort_Effects/Textures/waveH}
  n39 = TextureCoordinate()
  n40 = Abs(None=Add(A=ComponentMask(None=@n39) {g=true}) {constB=-0.5})
  n41 = TextureCoordinate()
  n42 = Abs(None=Add(A=ComponentMask(None=@n41) {r=true}) {constB=-0.5})
  n43 = ComponentMask(None=TextureCoordinate()) {g=true}
  n44 = Abs(None=Multiply(A=Add(A=@n43) {constB=-0.5}) {constB=2})
  n45 = LinearInterpolate(A=Multiply(A=@n42, B=@n42), B=Power(Base=@n42) {constExponent=0.7}, Alpha=Multiply(A=@n44, B=@n44))
  n46 = Multiply(A=DynamicParameter() {paramNames [PanPosition, IncomingAmplitude, ChromaPushAmount, Param4]}.PanPosition) {constB=-1}
  n47 = Add(A=@n45, B=@n46)
  n48 = ConstantBiasScale(None=Sine(None=Multiply(A=@n47) {constB=2}))
  n49 = Add(A=@n45, B=@n46)
  n50 = ConstantBiasScale(None=Cosine(None=Multiply(A=@n49) {constB=4}))
  n51 = Multiply(A=DynamicParameter() {paramNames [PanPosition, IncomingAmplitude, ChromaPushAmount, Param4]}.IncomingAmplitude) {constB=6}
  n52 = Multiply(
   A=Min(A=LinearInterpolate(Alpha=@n48) {constA=32}, B=LinearInterpolate(Alpha=@n50) {constA=32})
   B=LinearInterpolate(
    A=ScalarParameter() {defaultValue=12, parameterName=AmpHeightLow}
    B=ScalarParameter() {defaultValue=4.5, parameterName=AmpHeightHigh}
    Alpha=Saturate(None=@n51)
   ) {constA=8, constB=0.05}
  )
  n53 = Saturate(None=OneMinus(None=Multiply(A=@n40, B=@n52) {constB=2}))
  n54 = Multiply(A=@n53, B=@n53)
  n55 = Multiply(A=@n54, B=@n54)
  n56 = Multiply(A=@n55, B=@n55)
  n57 = Add(A=Multiply(A=Multiply(A=@n35.G, B=@n38.G), B=@n53), B=Multiply(A=@n56, B=@n56))
  n58 = Add(
   A=Multiply(A=Multiply(A=@n13.G, B=@n15), B=SphereMask(A=@n16, B=@n17, Radius=@n18, Hardness=@n19))
   B=Multiply(A=Saturate(None=@n24), B=Saturate(None=@n57))
  )
  Multiply(A=OneMinus(None=Saturate(None=@n6)), B=Multiply(A=@n1.A, B=Multiply(A=@n7, B=@n58)))
 ]
}
