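# Animation Blueprint for the M_MED_Lyrical backpack cosmetic, animating its globe with a spin and drifting float.
# Each update, Random Vector Function produces a drift vector and real time drives a steady yaw.
# The AnimGraph adds these to the globe bones alongside fixed tilt and lift, then blends in low-alpha, spring-based Anim Dynamics and nudges the pack bone.
# The chain starts from Input Pose; the Copy Pose From Mesh node is unconnected.
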AnimBlueprint /BRCosmetics/Accessories/FORT_Backpacks/Backpack_M_MED_Lyrical/Meshes/M_MED_Lyrical_Pack_AnimBP
target_skeleton=&/M_MED_Lyrical_Pack_Skeleton
graphs [
 AnimGraph {nodes [
   Copy Pose From Mesh() {bUseAttachedParent=true}
   n1 = "Transform (Modify) Bone - Bone: dyn_globe_local"(ComponentPose=Local To Component(LocalPose=Input Pose()), Alpha=1, Translation="0,0,0", Rotation="0,0,50", Scale="1,1,1") {rotationMode=BMM_Additive, alphaBoolBlend {blendOption=Linear}, boneToModify {boneName=dyn_globe_local}, rotation {roll=50}}
   n2 = "Transform (Modify) Bone - Bone: dyn_mstr_globe"(ComponentPose=@n1, Alpha=1, Translation="0,0,3", Rotation="0,0,0", Scale="1,1,1") {
    rotationSpace=BCS_WorldSpace
    translationMode=BMM_Additive
    translationSpace=BCS_WorldSpace
    alphaBoolBlend {blendOption=Linear}
    alphaScaleBiasClamp {interpSpeedDecreasing=2, interpSpeedIncreasing=2}
    boneToModify {boneName=dyn_mstr_globe}
    translation {z=3}
   }
   n3 = "Transform (Modify) Bone - Bone: dyn_globe_local"(ComponentPose=@n2, Alpha=1, Translation="0,0,0", Rotation=Get GlobeLocal() -> (GlobeLocal:Rotator), Scale="1,1,1") {rotationMode=BMM_Additive, rotationSpace=BCS_BoneSpace, alphaBoolBlend {blendOption=Linear}, boneToModify {boneName=dyn_globe_local}}
   n4 = "Transform (Modify) Bone - Bone: dyn_mstr_globe"(ComponentPose=@n3, Alpha=1, Translation=Get Globe movement() -> (Globe movement:Vector), Rotation="0,0,0", Scale="1,1,1") {
    rotationSpace=BCS_WorldSpace
    translationMode=BMM_Additive
    translationSpace=BCS_BoneSpace
    alphaBoolBlend {blendOption=Linear}
    alphaScaleBiasClamp {interpSpeedDecreasing=2, interpSpeedIncreasing=2}
    boneToModify {boneName=dyn_mstr_globe}
   }
   n5 = Anim Dynamics - Bone: dyn_globe_world(ComponentPose=@n4, Alpha=0.24, ExternalForce="0,0,0") {
    angularDampingOverride=1
    angularSpringConstant=20000
    bAngularSpring=true
    bGravityOverrideInSimSpace=true
    bLinearSpring=true
    bOverrideAngularDamping=true
    bOverrideLinearDamping=true
    bUseGravityOverride=true
    gravityScale=0.35
    linearDampingOverride=0.9999
    linearSpringConstant=50
    simulationSpace=RootRelative
    alphaBoolBlend {blendOption=Linear}
    boundBone {boneName=dyn_globe_world}
    componentAppliedLinearAccClamp=<30000 30000 30000>
    componentLinearAccScale=<0.3 0.3 0.4>
    componentLinearVelScale=<0.3 0.23 0.4>
    gravityOverride {z=-980}
    physicsBodyDefinitions [{
      sphereCollisionRadius=0
      boundBone {boneName=dyn_globe_world}
      boxExtents=<5 10 10>
      constraintSetup {angularLimitsMax=<0 0 0>, angularLimitsMin=<0 0 0>, angularTarget=<1 0 0>, linearAxesMax=<30 0 10>, linearAxesMin=<-30 -20 -25>}
      localJointOffset=<0 0 1>
     }]
   }
   n6 = "Transform (Modify) Bone - Bone: pack"(ComponentPose=@n5, Translation="0,-1.2,0", Rotation="0,0,0") {translationMode=BMM_Additive, alphaBoolBlend {blendOption=Linear}, boneToModify {boneName=pack}, translation {y=-1.2}}
   Output Pose(Result=Component To Local(ComponentPose=@n6)) {layerGroup=None}
  ]}
 EventGraph {nodes [
   Event BlueprintUpdateAnimation()
   > n1 = Sequence()
   > n2 = Random Vector Function!(
    execute
    CharacterPartAnimInstance:Custom Character Part Anim Instance Object Reference
    SkelMeshComponent:Mesh Component Object Reference=GetOwnerSkeletalMeshComponent()
    JointName:Name=None
    GameTimeInSeconds:double=GetGameTimeInSeconds()
    TimeOffset:double=0
    PerlinScale:double=0
    Frequency:Vector="0.9,1.1,0.75"
    Amplitude:Vector="0.5,0.7,0.5"
    Offset:Vector="0,0,0"
   ) -> (then, VectorOut:Vector)
   > Set Globe movement(Globe movement:Vector=@n2) -> (Output_Get:Vector)
   n3 = GetRealTimeSeconds()
   Set GlobeLocal(execute=@n1.then_1, GlobeLocal:Rotator=MakeRotator(Yaw=float * float(A=@n3, B=-20) {is=title K2Node_PromotableOperator})) -> (Output_Get:Rotator)
  ]}
]
variables [
 Globe movement {type=Vector}
 GlobeLocal {type=Rotator}
]
